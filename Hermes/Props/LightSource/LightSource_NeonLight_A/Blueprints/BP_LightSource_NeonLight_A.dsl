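; Neon light prop Blueprint that configures its mesh and neon material in the Construction Script.
; After the parent construction call, it picks a static mesh from the Meshes array by the Mesh index.
; With Override Color on, a dynamic material instance receives LightColor and EmissiveIntensity as Color and Brightness; otherwise slot 0 gets the valid ColorPreset material.
; BeginPlay, overlap and Tick events are empty.
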(event EventBeginPlay)

(event Collision|EventActorBeginOverlap
  (params (OtherActor :type Actor%20Object%20Reference)))

(event EventTick
  (params (DeltaSeconds :type Float%20%28single-precision%29)))

(fn ConstructionScript
  (bind _mesh (Variables|Default|GetMesh))
  (bind _meshes (Variables|Default|GetMeshes))
  (bind _output (Utilities|Array|Get(acopy) _meshes _mesh))
  (bind _colorpreset (Variables|Default|GetColorPreset))
  (bind _colorpresetmaterials (Variables|Default|GetColorPresetMaterials))
  (bind _output_2 (Utilities|Array|Get(acopy) _colorpresetmaterials _colorpreset))
  (|CallParentFunction)
  (if (Utilities|IsValid _output :!class /Script/Engine.KismetSystemLibrary)
    (Components|StaticMesh|SetStaticMesh (Variables|Debug|GetStaticMeshComponent) _output)
    (if (Variables|Default|GetOverride%20Color)
      (bind _returnvalue (Rendering|Material|CreateDynamicMaterialInstance (Variables|Debug|GetStaticMeshComponent) 0 0 "None" :!class /Script/Engine.PrimitiveComponent))
      (Rendering|Material|SetVectorParameterValue _returnvalue "Color" (Variables|Default|GetLightColor) :!class /Script/Engine.MaterialInstanceDynamic)
      (Rendering|Material|SetScalarParameterValue _returnvalue "Brightness" (Variables|Default|GetEmissiveIntensity) :!class /Script/Engine.MaterialInstanceDynamic)
      (elif (Utilities|IsValid _output_2 :!class /Script/Engine.KismetSystemLibrary)
        (Rendering|Material|SetMaterial (Variables|Debug|GetStaticMeshComponent) 0 _output_2)))))
